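# Material for a small plastic construction pallet prop, using the Default Lit shading model.
# It contains two texture samples from the Asteria construction pallet kit: one with a _D suffix, likely diffuse, and a normal map.
# The graph shows no links from either sample to material outputs and no parameters or math nodes.
# Automatic setting of usage flags in the editor is disabled.
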Material /SparksMaps/Environment/Props/ConstructionPalletPlastic/M_GR_PlasticPalletSmall
graph {
 properties {bAutomaticallySetUsageInEditor=false, usedShadingModels=Default Lit}
 nodes [
  TextureSample() {texture=/Game/Environments/Asteria/Props/Construction/Kit_Construction_Pallets_Plastic_A/Textures/T_Construction_Pallets_Plastic_Small_A_A_D}
  TextureSample() {
   samplerType=SAMPLERTYPE_Normal
   texture=/Game/Environments/Asteria/Props/Construction/Kit_Construction_Pallets_Plastic_A/Textures/T_Construction_Pallets_Plastic_Small_A_A_N
  }
 ]
}
